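# Post-process style Anim Blueprint for the large male BrawnGraze body cosmetic, adding secondary muscle jiggle to pecs, biceps and triceps.
# It copies the attached parent mesh's pose, then runs a Control Rig fed crouch and front-end flags.
# Two AnimDynamics chains branch from that cached pose: layered linear and angular springs for the front end, angular-only springs otherwise.
# bPbBlendEnabled switches instantly between the dynamics result and the plain Control Rig output.
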AnimBlueprint /BRCosmetics/Characters/Player/Male/Large/Bodies/M_LRG_BrawnGraze/Meshes/M_LRG_BrawnGraze_AnimBP
target_skeleton=&/M_LRG_BrawnGraze_Skeleton
graphs [
 AnimGraph {nodes [
   Output Pose(Result=Blend Poses by bool(
     BlendPose_0=Use cached pose 'BodyDynamics'()
     BlendPose_1=Use cached pose 'CtrlOutput'()
     BlendTime_0=0
     BlendTime_1=0
     bActiveValue=Get bPbBlendEnabled() -> (bPbBlendEnabled:bool)
    )) {layerGroup=None}
   n1 = Control Rig!(Source=Copy Pose From Mesh() {bUseAttachedParent=true}, Alpha, Crouching:bool=Get bIsCrouching(), Frontend:bool=Get bIsInFrontEnd()) -> (Pose) {controlRigAssetReference {blueprintRigClass=&/M_LRG_BrawnGraze_CtrlRig.M_LRG_BrawnGraze_CtrlRig_C}}
   Save cached pose 'CtrlOutput'(Pose=@n1) {cacheName=CtrlOutput}
   n2 = Anim Dynamics - Bone: deform_pec_l(ComponentPose=Local To Component(LocalPose=Use cached pose 'CtrlOutput'()), Alpha=0.3, ExternalForce="0,0,0") {
    angularDampingOverride=1
    angularSpringConstant=12000
    bLinearSpring=true
    bOverrideLinearDamping=true
    bUseGravityOverride=true
    gravityScale=0.35
    linearDampingOverride=1
    linearSpringConstant=700
    simulationSpace=World
    boundBone {boneName=deform_pec_l}
    componentAppliedLinearAccClamp=<3000 3000 3000>
    gravityOverride {z=-200}
    physicsBodyDefinitions [{
      boundBone {boneName=deform_pec_l}
      boxExtents=<50 50 50>
      constraintSetup {angularLimitsMax=<0 0 0>, angularLimitsMin=<0 0 0>, angularTarget=<1 0 0>, linearAxesMax=<0.3 0 0.2>, linearAxesMin=<-0.3 0 -0.2>}
      localJointOffset=<2 0 0>
     }]
    simSpaceSettings {maxAngularAcceleration=10, maxAngularVelocity=500, simSpaceAngularAlpha=0.5}
   }
   n3 = Anim Dynamics - Bone: deform_pec_r(ComponentPose=@n2, Alpha=0.3, ExternalForce="0,0,0") {
    angularDampingOverride=1
    angularSpringConstant=12000
    bLinearSpring=true
    bOverrideLinearDamping=true
    bUseGravityOverride=true
    gravityScale=0.35
    linearDampingOverride=1
    linearSpringConstant=700
    simulationSpace=World
    boundBone {boneName=deform_pec_r}
    componentAppliedLinearAccClamp=<3000 3000 3000>
    gravityOverride {z=-200}
    physicsBodyDefinitions [{
      boundBone {boneName=deform_pec_r}
      boxExtents=<50 50 50>
      constraintSetup {angularLimitsMax=<0 0 0>, angularLimitsMin=<0 0 0>, angularTarget=<1 0 0>, linearAxesMax=<0.3 0 0.2>, linearAxesMin=<-0.3 0 -0.2>}
      localJointOffset=<-2 0 0>
     }]
    simSpaceSettings {maxAngularAcceleration=10, maxAngularVelocity=500, simSpaceAngularAlpha=0.5}
   }
   n4 = Anim Dynamics - Bone: base_bicep_l(ComponentPose=@n3, Alpha=1, ExternalForce="0,0,0") {
    angularDampingOverride=0.9999
    angularSpringConstant=15000
    bAngularSpring=true
    bOverrideAngularDamping=true
    bUseGravityOverride=true
    gravityScale=0.35
    simulationSpace=RootRelative
    boundBone {boneName=base_bicep_l}
    componentAppliedLinearAccClamp=<3000 3000 3000>
    componentLinearAccScale=<1 1 1>
    componentLinearVelScale=<1 1 1>
    gravityOverride {z=-980}
    physicsBodyDefinitions [{
      boundBone {boneName=base_bicep_l}
      boxExtents=<30 30 30>
      constraintSetup {angularLimitsMax=<10 0 5>, angularLimitsMin=<-10 0 -5>, angularTarget=<1 0 0>}
      localJointOffset=<4 0 0>
     }]
    simSpaceSettings {maxAngularAcceleration=30, maxAngularVelocity=500, simSpaceAngularAlpha=0.7}
   }
   n5 = Anim Dynamics - Bone: base_bicep_l(ComponentPose=@n4, Alpha=0.8, ExternalForce="0,0,0") {
    angularDampingOverride=1
    angularSpringConstant=20000
    bLinearSpring=true
    bOverrideLinearDamping=true
    bUseGravityOverride=true
    gravityScale=0.35
    linearDampingOverride=0.999
    linearSpringConstant=200
    simulationSpace=RootRelative
    boundBone {boneName=base_bicep_l}
    componentAppliedLinearAccClamp=<2000 2000 2000>
    componentLinearAccScale=<1 1 1>
    componentLinearVelScale=<1 1 1>
    gravityOverride {z=-980}
    physicsBodyDefinitions [{
      boundBone {boneName=base_bicep_l}
      boxExtents=<40 40 40>
      constraintSetup {angularLimitsMax=<0 0 0>, angularLimitsMin=<0 0 0>, angularTarget=<1 0 0>, linearAxesMax=<0 0 1>, linearAxesMin=<0 0 -1>}
      localJointOffset=<4 0 0>
     }]
    simSpaceSettings {maxAngularAcceleration=20, maxAngularVelocity=300, simSpaceAngularAlpha=0.7}
   }
   n6 = Anim Dynamics - Bone: base_tricep_l(ComponentPose=@n5, Alpha=0.7, ExternalForce="0,0,0") {
    angularDampingOverride=1
    angularSpringConstant=20000
    bLinearSpring=true
    bOverrideLinearDamping=true
    bUseGravityOverride=true
    gravityScale=0.35
    linearDampingOverride=0.999
    linearSpringConstant=400
    simulationSpace=RootRelative
    boundBone {boneName=base_tricep_l}
    componentAppliedLinearAccClamp=<2000 2000 2000>
    componentLinearAccScale=<1 1 1>
    componentLinearVelScale=<1 1 1>
    gravityOverride {z=-980}
    physicsBodyDefinitions [{
      boundBone {boneName=base_tricep_l}
      boxExtents=<40 40 40>
      constraintSetup {angularLimitsMax=<0 0 0>, angularLimitsMin=<0 0 0>, angularTarget=<1 0 0>, linearAxesMax=<0 0 1>, linearAxesMin=<0 0 -1>}
      localJointOffset=<3 0 0>
     }]
    simSpaceSettings {maxAngularAcceleration=20, maxAngularVelocity=300, simSpaceAngularAlpha=0.8}
   }
   n7 = Anim Dynamics - Bone: base_tricep_l(ComponentPose=@n6, Alpha=1, ExternalForce="0,0,0") {
    angularDampingOverride=0.9999
    angularSpringConstant=15000
    bAngularSpring=true
    bOverrideAngularDamping=true
    bUseGravityOverride=true
    gravityScale=0.35
    simulationSpace=RootRelative
    boundBone {boneName=base_tricep_l}
    componentAppliedLinearAccClamp=<2000 2000 2000>
    componentLinearAccScale=<0.5 0.5 0.5>
    componentLinearVelScale=<0.5 0.5 0.5>
    gravityOverride {z=-980}
    physicsBodyDefinitions [{
      boundBone {boneName=base_tricep_l}
      boxExtents=<30 30 30>
      constraintSetup {angularLimitsMax=<15 15 15>, angularLimitsMin=<-15 -15 -15>, angularTarget=<1 0 0>}
      localJointOffset=<3 0 0>
     }]
    simSpaceSettings {maxAngularAcceleration=10, maxAngularVelocity=300, simSpaceAngularAlpha=0.7}
   }
   n8 = Anim Dynamics - Bone: base_bicep_r(ComponentPose=@n7, Alpha=1, ExternalForce="0,0,0") {
    angularDampingOverride=0.9999
    angularSpringConstant=15000
    bAngularSpring=true
    bOverrideAngularDamping=true
    bUseGravityOverride=true
    gravityScale=0.35
    simulationSpace=RootRelative
    boundBone {boneName=base_bicep_r}
    componentAppliedLinearAccClamp=<3000 3000 3000>
    componentLinearAccScale=<1 1 1>
    componentLinearVelScale=<1 1 1>
    gravityOverride {z=-980}
    physicsBodyDefinitions [{
      boundBone {boneName=base_bicep_r}
      boxExtents=<30 30 30>
      constraintSetup {angularLimitsMax=<10 0 5>, angularLimitsMin=<-10 0 -5>, angularTarget=<1 0 0>}
      localJointOffset=<-4 0 0>
     }]
    simSpaceSettings {maxAngularAcceleration=30, maxAngularVelocity=500, simSpaceAngularAlpha=0.7}
   }
   n9 = Anim Dynamics - Bone: base_bicep_r(ComponentPose=@n8, Alpha=0.8, ExternalForce="0,0,0") {
    angularDampingOverride=1
    angularSpringConstant=20000
    bLinearSpring=true
    bOverrideLinearDamping=true
    bUseGravityOverride=true
    gravityScale=0.35
    linearDampingOverride=0.999
    linearSpringConstant=200
    simulationSpace=RootRelative
    boundBone {boneName=base_bicep_r}
    componentAppliedLinearAccClamp=<2000 2000 2000>
    componentLinearAccScale=<1 1 1>
    componentLinearVelScale=<1 1 1>
    gravityOverride {z=-980}
    physicsBodyDefinitions [{
      boundBone {boneName=base_bicep_r}
      boxExtents=<40 40 40>
      constraintSetup {angularLimitsMax=<0 0 0>, angularLimitsMin=<0 0 0>, angularTarget=<1 0 0>, linearAxesMax=<0 0 1>, linearAxesMin=<0 0 -1>}
      localJointOffset=<-4 0 0>
     }]
    simSpaceSettings {maxAngularAcceleration=20, maxAngularVelocity=300, simSpaceAngularAlpha=0.7}
   }
   n10 = Anim Dynamics - Bone: base_tricep_r(ComponentPose=@n9, Alpha=0.7, ExternalForce="0,0,0") {
    angularDampingOverride=1
    angularSpringConstant=20000
    bLinearSpring=true
    bOverrideLinearDamping=true
    bUseGravityOverride=true
    gravityScale=0.35
    linearDampingOverride=0.999
    linearSpringConstant=400
    simulationSpace=RootRelative
    boundBone {boneName=base_tricep_r}
    componentAppliedLinearAccClamp=<2000 2000 2000>
    componentLinearAccScale=<1 1 1>
    componentLinearVelScale=<1 1 1>
    gravityOverride {z=-980}
    physicsBodyDefinitions [{
      boundBone {boneName=base_tricep_r}
      boxExtents=<40 40 40>
      constraintSetup {angularLimitsMax=<0 0 0>, angularLimitsMin=<0 0 0>, angularTarget=<1 0 0>, linearAxesMax=<0 0 1>, linearAxesMin=<0 0 -1>}
      localJointOffset=<-3 0 0>
     }]
    simSpaceSettings {maxAngularAcceleration=20, maxAngularVelocity=300, simSpaceAngularAlpha=0.8}
   }
   n11 = Anim Dynamics - Bone: base_tricep_r(ComponentPose=@n10, Alpha=1, ExternalForce="0,0,0") {
    angularDampingOverride=0.9999
    angularSpringConstant=15000
    bAngularSpring=true
    bOverrideAngularDamping=true
    bUseGravityOverride=true
    gravityScale=0.35
    simulationSpace=RootRelative
    boundBone {boneName=base_tricep_r}
    componentAppliedLinearAccClamp=<2000 2000 2000>
    componentLinearAccScale=<0.5 0.5 0.5>
    componentLinearVelScale=<0.5 0.5 0.5>
    gravityOverride {z=-980}
    physicsBodyDefinitions [{
      boundBone {boneName=base_tricep_r}
      boxExtents=<30 30 30>
      constraintSetup {angularLimitsMax=<15 15 15>, angularLimitsMin=<-15 -15 -15>, angularTarget=<1 0 0>}
      localJointOffset=<-3 0 0>
     }]
    simSpaceSettings {maxAngularAcceleration=10, maxAngularVelocity=300, simSpaceAngularAlpha=0.7}
   }
   n12 = Anim Dynamics - Bone: deform_pec_l(ComponentPose=Local To Component(LocalPose=Use cached pose 'CtrlOutput'()), Alpha=1, ExternalForce="0,0,0") {
    angularDampingOverride=1
    angularSpringConstant=20000
    bAngularSpring=true
    bOverrideAngularDamping=true
    bUseGravityOverride=true
    gravityScale=0.35
    simulationSpace=RootRelative
    boundBone {boneName=deform_pec_l}
    componentAppliedLinearAccClamp=<3000 3000 3000>
    componentLinearAccScale=<1 1 1>
    componentLinearVelScale=<1 1 1>
    gravityOverride {z=-980}
    physicsBodyDefinitions [{
      boundBone {boneName=deform_pec_l}
      boxExtents=<20 30 30>
      constraintSetup {angularLimitsMax=<0 2 2>, angularLimitsMin=<0 -2 -2>, angularTarget=<1 0 0>}
      localJointOffset=<4 0 0>
     }]
    simSpaceSettings {simSpaceAngularAlpha=0.5}
   }
   n13 = Anim Dynamics - Bone: deform_pec_r(ComponentPose=@n12, Alpha=1, ExternalForce="0,0,0") {
    angularDampingOverride=1
    angularSpringConstant=20000
    bAngularSpring=true
    bOverrideAngularDamping=true
    bUseGravityOverride=true
    gravityScale=0.35
    simulationSpace=RootRelative
    boundBone {boneName=deform_pec_r}
    componentAppliedLinearAccClamp=<3000 3000 3000>
    componentLinearAccScale=<1 1 1>
    componentLinearVelScale=<1 1 1>
    gravityOverride {z=-980}
    physicsBodyDefinitions [{
      boundBone {boneName=deform_pec_r}
      boxExtents=<20 30 30>
      constraintSetup {angularLimitsMax=<0 2 2>, angularLimitsMin=<0 -2 -2>, angularTarget=<1 0 0>}
      localJointOffset=<4 0 0>
     }]
    simSpaceSettings {simSpaceAngularAlpha=0.5}
   }
   n14 = Anim Dynamics - Bone: base_bicep_l(ComponentPose=@n13, Alpha=1, ExternalForce="0,0,0") {
    angularDampingOverride=1
    angularSpringConstant=20000
    bAngularSpring=true
    bOverrideAngularDamping=true
    bUseGravityOverride=true
    gravityScale=0.35
    simulationSpace=RootRelative
    boundBone {boneName=base_bicep_l}
    componentAppliedLinearAccClamp=<3000 3000 3000>
    componentLinearAccScale=<1 1 1>
    componentLinearVelScale=<1 1 1>
    gravityOverride {z=-980}
    physicsBodyDefinitions [{
      boundBone {boneName=base_bicep_l}
      boxExtents=<20 30 30>
      constraintSetup {angularLimitsMax=<10 0 5>, angularLimitsMin=<-10 0 -5>, angularTarget=<1 0 0>}
      localJointOffset=<4 0 0>
     }]
    simSpaceSettings {simSpaceAngularAlpha=0.7}
   }
   n15 = Anim Dynamics - Bone: base_tricep_l(ComponentPose=@n14, Alpha=1, ExternalForce="0,0,0") {
    angularDampingOverride=1
    angularSpringConstant=20000
    bAngularSpring=true
    bOverrideAngularDamping=true
    bUseGravityOverride=true
    gravityScale=0.35
    simulationSpace=RootRelative
    boundBone {boneName=base_tricep_l}
    componentLinearAccScale=<1 1 1>
    componentLinearVelScale=<1 1 1>
    gravityOverride {z=-980}
    physicsBodyDefinitions [{
      boundBone {boneName=base_tricep_l}
      boxExtents=<20 30 30>
      constraintSetup {angularLimitsMax=<15 15 15>, angularLimitsMin=<-15 -15 -15>, angularTarget=<1 0 0>}
      localJointOffset=<4 0 0>
     }]
   }
   n16 = Anim Dynamics - Bone: base_bicep_r(ComponentPose=@n15, Alpha=1, ExternalForce="0,0,0") {
    angularDampingOverride=1
    angularSpringConstant=20000
    bAngularSpring=true
    bOverrideAngularDamping=true
    bUseGravityOverride=true
    gravityScale=0.35
    simulationSpace=RootRelative
    boundBone {boneName=base_bicep_r}
    componentAppliedLinearAccClamp=<3000 3000 3000>
    componentLinearAccScale=<1 1 1>
    componentLinearVelScale=<1 1 1>
    gravityOverride {z=-980}
    physicsBodyDefinitions [{
      boundBone {boneName=base_bicep_r}
      boxExtents=<20 30 30>
      constraintSetup {angularLimitsMax=<10 0 5>, angularLimitsMin=<-10 0 -5>, angularTarget=<1 0 0>}
      localJointOffset=<4 0 0>
     }]
    simSpaceSettings {simSpaceAngularAlpha=0.7}
   }
   n17 = Anim Dynamics - Bone: base_tricep_r(ComponentPose=@n16, Alpha=1, ExternalForce="0,0,0") {
    angularDampingOverride=1
    angularSpringConstant=20000
    bAngularSpring=true
    bOverrideAngularDamping=true
    bUseGravityOverride=true
    gravityScale=0.35
    simulationSpace=RootRelative
    boundBone {boneName=base_tricep_r}
    componentLinearAccScale=<1 1 1>
    componentLinearVelScale=<1 1 1>
    gravityOverride {z=-980}
    physicsBodyDefinitions [{
      boundBone {boneName=base_tricep_r}
      boxExtents=<20 30 30>
      constraintSetup {angularLimitsMax=<15 15 15>, angularLimitsMin=<-15 -15 -15>, angularTarget=<1 0 0>}
      localJointOffset=<4 0 0>
     }]
   }
   Save cached pose 'BodyDynamics'(Pose=Blend Poses by bool(
     BlendPose_0=Component To Local(ComponentPose=@n11)
     BlendPose_1=Component To Local(ComponentPose=@n17)
     BlendTime_0=0
     BlendTime_1=0
     bActiveValue=Get bIsInFrontEnd()
    )) {cacheName=BodyDynamics}
  ]}
 EventGraph {links [], nodes []}
]
variables []
