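# Physics asset of cape collision shapes for the large male BrawnGraze player body, evidently for cloth or cape simulation to collide against.
# Overlapping capsules cover the pelvis, spine, clavicles, neck and feet, with a sphere on the head.
# Extra pelvis capsules, offset along the bone axis, widen coverage beyond the hip joint.
# Arm, thigh and calf bodies hold no shapes, and no constraints link the bodies.
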PhysicsAsset /BRCosmetics/Characters/Player/Male/Large/Bodies/M_LRG_BrawnGraze/Meshes/M_LRG_BrawnGraze_CapeColliders
bodies {
 calf_l []
 calf_r []
 clavicle_l [{shapeType=Capsule, length=10.1074, radius=12.299, center=<11.3652 -1.3589 0.6547>, rotation=<rot 75.9365 -0.0001 -180>}]
 clavicle_r [{shapeType=Capsule, length=10.1074, radius=12.299, center=<-11.3652 1.3589 -0.6547>, rotation=<rot 75.9364 -0.0001 0>}]
 foot_l_01 [{
   shapeType=Capsule
   length=15.3634
   radius=12.5255
   shapeName=foot_l_01_capsule
   center=<-2.6431 9.9517 3.4186>
   rotation=<rot 0 178.641 93.3844>
  }]
 foot_r_01 [{
   shapeType=Capsule
   length=15.3634
   radius=12.5255
   shapeName=foot_r_01_capsule
   center=<2.6431 -9.9517 -3.4186>
   rotation=<rot 0 178.641 -86.6156>
  }]
 head [{shapeType=Sphere, radius=17.3125, center=<11.028 -0.174 0>}]
 lowerarm_l []
 lowerarm_r []
 neck_02 [{shapeType=Capsule, length=11.25, radius=9.505, center=<0.6451 -3.6755 0>, rotation=<rot 90 10.6197 6.3412>}]
 pelvis [
  {shapeType=Capsule, length=12.4687, radius=15.195, center=<-8.1326 -5.0182 0>}
  {shapeType=Capsule, length=4.0164, radius=25.9836, shapeName=pelvis_capsule, center=<-44.9307 -10.3638 0>}
  {shapeType=Capsule, length=15, radius=15, shapeName=pelvis_capsule, center=<-21.844 -11.9023 0>}
 ]
 spine_01 [{shapeType=Capsule, length=10.0937, radius=15.9763, center=<-2.7814 0.7831 0>}]
 spine_02 [
  {shapeType=Capsule, length=5.0937, radius=17.2888, center=<-3.0582 1.3295 0.0001>}
  {shapeType=Capsule, length=26.4687, radius=14.5075, center=<2.5024 0.4992 3.9825>, rotation=<rot 90 0.0049 0.0049>}
  {shapeType=Capsule, length=25.6875, radius=15.2888, center=<2.4704 0.4991 -3.9825>, rotation=<rot 90 -0.0396 -0.0396>}
 ]
 spine_03 [{shapeType=Capsule, length=6.0312, radius=17.4763, center=<-0.6979 -0.218 0.0001>}]
 spine_04 [{shapeType=Capsule, length=15.0625, radius=18.3825, center=<-0.6509 3.7251 0.0002>}]
 spine_05 [
  {shapeType=Capsule, length=22.2812, radius=16.445, center=<-3.5391 -0.9968 0>}
  {shapeType=Capsule, length=15.5938, radius=13.625, center=<3.6493 -1.7428 0>}
 ]
 thigh_l []
 thigh_r []
 upperarm_l []
 upperarm_r []
}
constraints []
