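# Control Rig for the BrawnGraze large male face accessory, adjusting a hat and its bowknot bones procedurally.
# When NotEmote is true, a spherical pose reader on the head bone outputs a weight that blends in corrective transforms on the hat bones.
# ParachuteOpened, smoothed through an alpha interpolation, weights an additional rotation on the dyn_hatBowknot_03 bone.
# The pose reader has debug drawing enabled.
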ControlRigBlueprint /BRCosmetics/Characters/Player/Male/Large/Bodies/M_LRG_BrawnGraze/Meshes/Parts/M_LRG_BrawnGraze_FaceAcc_CtrlRig
graphs [
 RigVMFunctionLibrary {links [], nodes []}
 RigVMModel {nodes [
   RigUnit_BeginExecution = &RigUnit_BeginExecution()
   RigVMFunction_Sequence = &RigVMFunction_Sequence(ExecuteContext=@RigUnit_BeginExecution.ExecutePin)
   n1 = &RigVMFunction_ControlFlowBranch(ExecuteContext=@RigVMFunction_Sequence.A, Condition=RigVMVariableNode() {variable=NotEmote}.Value)
   n2 = &RigUnit_SphericalPoseReader(
    ActiveRegionScaleFactors=(PositiveWidth=1,NegativeWidth=1,PositiveHeight=1,NegativeHeight=1)
    ActiveRegionSize=0.1
    Debug=(bDrawDebug=true,bDraw2D=False,bDrawLocalAxes=False,DebugScale=25,DebugSegments=20,DebugThickness=0.25)
    DriverAxis=(X=1,Y=0,Z=0)
    DriverItem=(Type=Bone,Name="head")
    FalloffRegionScaleFactors=(PositiveWidth=0.5,NegativeWidth=1,PositiveHeight=1,NegativeHeight=1)
    FalloffSize=0.3
    FlipHeightScaling=False
    FlipWidthScaling=False
    OptionalParentItem=(Type=Bone,Name="None")
    RotationOffset=(X=-70,Y=0,Z=40)
    ExecutePin=@n1.True
   )
   > &RigUnit_ModifyTransforms(
    ItemToModify="((Item=(Type=Bone,Name=\"base_hat\"),Transform=(Rotation=(X=0.0039,Y=0.07,Z=-0.1525,W=0.9858),Translation=(X=-0.8,Y=0.1859,Z=0.1451),Scale3D=(X=1,Y=1,Z=1))),(Item=(Type=Bone,Name=\"base_hatBowknot_03\"),Transform=(Rotation=(X=-0.0993,Y=-0.2348,Z=-0.0077,W=0.96 ...(truncated)"
    WeightMinimum=0
    Weight=@n2.OutputParam
   )
   AlphaInterp = &RigVMFunction_AlphaInterp(InterpSpeedDecreasing=10, InterpSpeedIncreasing=10, Scale=1, bClampResult=False, Value=RigVMVariableNode() {variable=ParachuteOpened}.Value)
   > &RigUnit_ModifyTransforms(
    ItemToModify=((Item=(Type=Bone,Name="dyn_hatBowknot_03"),Transform=(Rotation=(X=0,Y=-0.1736,Z=0,W=0.9848),Translation=(X=0,Y=0,Z=0),Scale3D=(X=1,Y=1,Z=1))))
    WeightMinimum=0
    Weight=@AlphaInterp.Result
   )
  ]}
]
hierarchy [
 root {loc=<0 0 0>, rot=<0 0 0 1>}
 pelvis {parent=root, loc=<0 -0.7321 92.4218>, rot=<0.0127 0.707 -0.0127 -0.707>}
 spine_01 {parent=pelvis, loc=<5.6002 -0.5128 0>, rot=<0 0 0.034 -0.9994>}
 spine_02 {parent=spine_01, loc=<8.0756 0.0139 0>, rot=<0 0 0.0238 -0.9997>}
 spine_03 {parent=spine_02, loc=<7.0395 0 0>, rot=<0 0 0.0149 -0.9999>}
 spine_04 {parent=spine_03, loc=<8.1023 0 0>, rot=<0 0 -0.0132 -0.9999>}
 spine_05 {parent=spine_04, loc=<10.0141 0 0>, rot=<0 0 -0.0015 -1>}
 clavicle_l {parent=spine_05, loc=<2.5856 -0.2866 -4.3777>, rot=<0.0465 -0.7289 -0.009 -0.6829>}
 upperarm_l {parent=clavicle_l, loc=<18.2264 0.1342 -0.001>, rot=<0.0253 -0.2693 0.0013 -0.9627>}
 neck_01 {parent=spine_05, loc=<10.815 1.9196 0>, rot=<0 0 -0.1864 -0.9825>}
 neck_02 {parent=neck_01, loc=<4.7347 0.3045 0>, rot=<0 0 0.0363 -0.9993>}
 head {parent=neck_02, loc=<5.7632 0.5552 0>, rot=<0 0 0.115 -0.9934>}
 base_hat {parent=head, loc=<16.5271 -2.2574 0.0004>, rot=<0.0041 -1 0 -0.0002>}
 base_hatBowknot_02 {parent=base_hat, loc=<-0.3779 13.8107 0.0065>, rot=<0.5868 0.3944 -0.3942 -0.5871>}
 dyn_hatBowknot_02 {parent=base_hatBowknot_02, loc=<0 0 0>, rot=<0 0 0 1>}
 base_hatBowknot_01 {parent=base_hat, loc=<-0.6182 0 12.7598>, rot=<-0.0397 0.6562 -0.0452 -0.7522>}
 dyn_hatBowknot_01 {parent=base_hatBowknot_01, loc=<0 0 0>, rot=<0 0 0 1>}
 base_hatBowknot_03 {parent=base_hat, loc=<2.0886 -11.1325 -0.0506>, rot=<-0.6157 0.3479 0.3375 -0.6213>}
 dyn_hatBowknot_03 {parent=base_hatBowknot_03, loc=<0 0 0>, rot=<0 0 0 1>}
 base_hatBowknot_04 {parent=base_hat, loc=<-0.626 0 -12.7595>, rot=<0.6564 0.0396 0.752 -0.0452>}
 dyn_hatBowknot_04 {parent=base_hatBowknot_04, loc=<0 0 0>, rot=<0 0 0 -1>}
 clavicle_r {parent=spine_05, loc=<2.5855 -0.2866 4.3777>, rot=<0.7289 0.0465 -0.6829 0.009>}
 upperarm_r {parent=clavicle_r, loc=<-18.2264 -0.1342 0.001>, rot=<0.0253 -0.2693 0.0013 -0.9627>}
]
variables [
 NotEmote {type=bool, default=False}
 ParachuteOpened {type=double, default=0}
]
