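# Animation Blueprint for the BrawnGraze large male face accessory, adding secondary motion to a hat bowknot.
# It copies the attached parent's pose into a Control Rig fed by emote and parachute states, then chains sprung, head-relative AnimDynamics on four bowknot bones.
# A per-bone layered blend applies DefaultSlot animation to the hat bones, weighted by the front-end instance's Frontend_Idle_Bypass curve.
# The event graph updates these state and weight variables.
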AnimBlueprint /BRCosmetics/Characters/Player/Male/Large/Bodies/M_LRG_BrawnGraze/Meshes/Parts/M_LRG_BrawnGraze_FaceAcc_AnimBP
target_skeleton=&/M_LRG_BrawnGraze_FaceAcc_Skeleton
graphs [
 AnimGraph {nodes [
   n1 = Control Rig!(
    Source=Copy Pose From Mesh() {bUseAttachedParent=true}
    Alpha
    NotEmote:bool=Get NotEmote() -> (NotEmote:bool)
    ParachuteOpened:double=Get ParachuteOpened() -> (ParachuteOpened:double)
   ) -> (Pose) {controlRigAssetReference {blueprintRigClass=&/M_LRG_BrawnGraze_FaceAcc_CtrlRig.M_LRG_BrawnGraze_FaceAcc_CtrlRig_C}}
   n2 = Anim Dynamics - Bone: dyn_hatBowknot_04(ComponentPose=Local To Component(LocalPose=@n1), Alpha=1, ExternalForce="0,0,0") {
    angularDampingOverride=1
    angularSpringConstant=20000
    bAngularSpring=true
    bOverrideAngularDamping=true
    bUseGravityOverride=true
    gravityScale=0.35
    simulationSpace=BoneRelative
    boundBone {boneName=dyn_hatBowknot_04}
    componentAppliedLinearAccClamp=<2000 2000 2000>
    componentLinearAccScale=<0.2 0.2 0.2>
    componentLinearVelScale=<0.2 0.2 0.2>
    gravityOverride {z=-980}
    physicsBodyDefinitions [{
      boundBone {boneName=dyn_hatBowknot_04}
      boxExtents=<20 30 30>
      constraintSetup {angularLimitsMax=<0 5 0>, angularLimitsMin=<0 -5 0>, angularTarget=<1 0 0>}
      localJointOffset=<-2 0 0>
     }]
    relativeSpaceBone {boneName=head}
   }
   n3 = Anim Dynamics - Bone: dyn_hatBowknot_03(ComponentPose=@n2, Alpha=1, ExternalForce="0,0,0") {
    angularDampingOverride=1
    angularSpringConstant=20000
    bAngularSpring=true
    bOverrideAngularDamping=true
    bUseGravityOverride=true
    gravityScale=0.35
    simulationSpace=BoneRelative
    boundBone {boneName=dyn_hatBowknot_03}
    componentAppliedLinearAccClamp=<2000 2000 2000>
    componentLinearAccScale=<0.2 0.2 0.2>
    componentLinearVelScale=<0.2 0.2 0.2>
    gravityOverride {z=-980}
    physicsBodyDefinitions [{
      boundBone {boneName=dyn_hatBowknot_03}
      boxExtents=<20 30 30>
      constraintSetup {angularLimitsMax=<0 5 0>, angularLimitsMin=<0 -5 0>, angularTarget=<1 0 0>}
      localJointOffset=<2 0 0>
     }]
    relativeSpaceBone {boneName=head}
   }
   n4 = Anim Dynamics - Bone: dyn_hatBowknot_01(ComponentPose=@n3, Alpha=1, ExternalForce="0,0,0") {
    angularDampingOverride=1
    angularSpringConstant=20000
    bAngularSpring=true
    bOverrideAngularDamping=true
    bUseGravityOverride=true
    gravityScale=0.35
    simulationSpace=BoneRelative
    boundBone {boneName=dyn_hatBowknot_01}
    componentAppliedLinearAccClamp=<2000 2000 2000>
    componentLinearAccScale=<0.2 0.2 0.2>
    componentLinearVelScale=<0.2 0.2 0.2>
    gravityOverride {z=-980}
    physicsBodyDefinitions [{
      boundBone {boneName=dyn_hatBowknot_01}
      boxExtents=<20 30 30>
      constraintSetup {angularLimitsMax=<0 5 0>, angularLimitsMin=<0 -5 0>, angularTarget=<1 0 0>}
      localJointOffset=<2 0 0>
     }]
    relativeSpaceBone {boneName=head}
   }
   n5 = Anim Dynamics - Bone: dyn_hatBowknot_02(ComponentPose=@n4, Alpha=1, ExternalForce="0,0,0") {
    angularDampingOverride=1
    angularSpringConstant=20000
    bAngularSpring=true
    bOverrideAngularDamping=true
    bUseGravityOverride=true
    gravityScale=0.35
    simulationSpace=BoneRelative
    boundBone {boneName=dyn_hatBowknot_02}
    componentAppliedLinearAccClamp=<2000 2000 2000>
    componentLinearAccScale=<0.2 0.2 0.2>
    componentLinearVelScale=<0.2 0.2 0.2>
    gravityOverride {z=-980}
    physicsBodyDefinitions [{
      boundBone {boneName=dyn_hatBowknot_02}
      boxExtents=<20 30 30>
      constraintSetup {angularLimitsMax=<0 5 0>, angularLimitsMin=<0 -5 0>, angularTarget=<1 0 0>}
      localJointOffset=<2 0 0>
     }]
    relativeSpaceBone {boneName=head}
   }
   n6 = Layered blend per bone(
    BasePose=Component To Local(ComponentPose=@n5)
    BlendPoses_0=Slot 'DefaultSlot'()
    BlendWeights_0=Get Custom Idle Is Playing Float() -> (Custom Idle Is Playing Float:double)
   ) {layerSetup [{branchFilters [
       {boneName=base_hat}
       {boneName=base_hatBowknot_01}
       {boneName=dyn_hatBowknot_01}
       {boneName=base_hatBowknot_02}
       {boneName=dyn_hatBowknot_02}
       {boneName=base_hatBowknot_03}
       {boneName=dyn_hatBowknot_03}
       {boneName=base_hatBowknot_04}
       {boneName=dyn_hatBowknot_04}
      ]}]}
   Output Pose(Result=@n6) {layerGroup=None}
  ]}
 EventGraph {nodes [
   Event BlueprintUpdateAnimation()
   > n1 = Sequence!(execute) -> (then_0, then_1, then_2:Exec, then_3:Exec)
   > Set NotEmote(NotEmote:bool=NOT Boolean(A=Get bIsPlayingEmote())) -> (Output_Get:bool)
   > Branch(Condition=Get bIsInFrontEnd())
   > Set Custom Idle Is Playing Float(Custom Idle Is Playing Float:double=GetCurveValue(self=GetOwnerFrontEndAnimInstance(), CurveName=Frontend_Idle_Bypass)) -> (Output_Get:double)
   n2 = Get bIsPlayingEmote()
   > n3 = Branch(Condition=AND Boolean(A=Get bIsInFrontEnd(), B=NOT Boolean(A=@n2)))
   > Set Is Front End(Is Front End:double=1) -> (Output_Get:double)
   Set Is Front End(execute=@n3.else, Is Front End=0)
   Set ParachuteOpened(execute=@n1.then_1, ParachuteOpened:double="To Float (Boolean)"(InBool=Get bIsParachuteOpened())) -> (Output_Get:double)
  ]}
]
variables [
 Custom Idle Is Playing Float {type=number}
 Is Front End {type=number}
 NotEmote {type=boolean}
 ParachuteOpened {type=number}
]
